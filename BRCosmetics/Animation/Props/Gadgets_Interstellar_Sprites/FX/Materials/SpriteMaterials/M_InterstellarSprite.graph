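# Masked, two-sided lit material for Niagara sprites drawn from the Interstellar sprite sheet, with optional palette recoloring.
# Thresholds on the sheet's green channel build wing and face masks, which appear to split wing, face and eye regions.
# Each region has a blue, purple and red color set, blended by the PaletteSelection dynamic parameter through Lerp_3Color.
# The UsePaletteColors static switch chooses between a rerouted input, likely the palettized result, and sheet color times particle color.
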Material /BRCosmetics/Animation/Props/Gadgets_Interstellar_Sprites/FX/Materials/SpriteMaterials/M_InterstellarSprite
graph {
 properties {
  bUsedWithNiagaraSprites=true
  blendMode=BLEND_Masked
  opacityMaskClipValue=0.1
  pixelDepthOffsetMode=PDOM_Legacy
  twoSided=true
  usedShadingModels=Default Lit
 }
 nodes [
  n1 = TextureSampleParameter2D() {parameterName=Sheet, texture=/BRCosmetics/Animation/Props/Gadgets_Interstellar_Sprites/FX/Textures/T_InterstellarSheet}
  n2 = Constant() {r=1}
  NamedRerouteDeclaration(None=If(A=@n1.G, B=Constant() {r=0.1}, "A > B"=Constant(), "A == B"=@n2, "A < B"=@n2)) {name=WingMask, nodeColor=<rgba 0 0.1562 1 1>}
  n3 = Constant() {r=1}
  NamedRerouteDeclaration(None=If(A=@n1.G, B=Constant() {r=0.6}, "A > B"=@n3, "A == B"=@n3, "A < B"=Constant())) {name=FaceMask, nodeColor=<rgba 0.2891 1 0 1>}
  n4 = Multiply(A=NamedRerouteUsage(), B=NamedRerouteUsage())
  n5 = Multiply(A=NamedRerouteUsage(), B=NamedRerouteUsage())
  n6 = Multiply(A=OneMinus(None=NamedRerouteUsage()), B=OneMinus(None=NamedRerouteUsage()))
  n7 = NamedRerouteUsage()
  NamedRerouteDeclaration(None=Add(A=Add(A=@n4, B=@n5), B=Multiply(A=@n6, B=@n7))) {name=PalettizedColors, nodeColor=<rgba 0 0.625 1 1>}
  NamedRerouteDeclaration(None=MaterialFunctionCall(
    A=VectorParameter() {parameterName=WingColor_Blue, defaultValue=<rgba 0.0194 0.0666 0.3372 1>}.RGB
    B=VectorParameter() {parameterName=WingColor_Purple, defaultValue=<rgba 0.0194 0.0666 0.3372 1>}.RGB
    C=VectorParameter() {parameterName=WingColor_Red, defaultValue=<rgba 0.0194 0.117 0.4179 1>}.RGB
    Alpha=DynamicParameter() {defaultValue=<rgba 0 1 1 4>, paramNames [SubUVIndex, PaletteSelection, ParticleUV_y, NumSprites]}.PaletteSelection
   ) {materialFunction=&Lerp_3Color}.Result) {name=Pal_Wing, nodeColor=<rgba 0 1 0.6719 1>}
  NamedRerouteDeclaration(None=MaterialFunctionCall(
    A=VectorParameter() {parameterName=FaceColor_Blue, defaultValue=<rgba 0.0194 0.117 0.4179 1>}.RGB
    B=VectorParameter() {parameterName=FaceColor_Purple, defaultValue=<rgba 0.4988 0.117 0.626 1>}.RGB
    C=VectorParameter() {parameterName=FaceColor_Red, defaultValue=<rgba 0.7991 0.0176 0.0091 1>}.RGB
    Alpha=DynamicParameter() {defaultValue=<rgba 0 1 1 4>, paramNames [SubUVIndex, PaletteSelection, ParticleUV_y, NumSprites]}.PaletteSelection
   ) {materialFunction=&Lerp_3Color}.Result) {name=Pal_Face, nodeColor=<rgba 1 0 0.0469 1>}
  NamedRerouteDeclaration(None=MaterialFunctionCall(
    A=VectorParameter() {parameterName=EyeColor_Blue, defaultValue=<rgba 1 0 0 1>}.RGB
    B=VectorParameter() {parameterName=EyeColor_Purple, defaultValue=<rgba 1 0 0 1>}.RGB
    C=VectorParameter() {parameterName=EyeColor_Red, defaultValue=<rgba 0.8714 0.8308 0.0595 1>}.RGB
    Alpha=DynamicParameter() {defaultValue=<rgba 0 1 1 4>, paramNames [SubUVIndex, PaletteSelection, ParticleUV_y, NumSprites]}.PaletteSelection
   ) {materialFunction=&Lerp_3Color}.Result) {name=Pal_Eye, nodeColor=<rgba 0 0.3906 1 1>}
  StaticSwitchParameter(True=NamedRerouteUsage(), False=Multiply(A=@n1.RGB, B=ParticleColor().RGB)) {parameterName=UsePaletteColors}
 ]
}
